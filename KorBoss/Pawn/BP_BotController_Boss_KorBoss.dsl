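; Bot controller Blueprint for the KorBoss boss, managing how its random targeter component may change targets.
; On BeginPlay it calls the parent, then binds to the perception event for a change of current target.
; Each target change sets CanSwitchTarget to false on the KorBoss AIRandomTargeter component; nothing in this controller turns it back on.
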(event EventBeginPlay
  (|CallParentFunction)
  (Perception|BindEventtoOnCurrentTargetChangedEvent (delegate Custom|TargetJustChanged)))

(event Custom|TargetJustChanged
  (params (OldTarget :type Actor%20Object%20Reference) (NewTarget :type Actor%20Object%20Reference))
  (Variables|/KorBoss/Pawn/FortControllerComponent_AIRandomTargeter_KorBoss.FortControllerComponent_AIRandomTargeter_KorBoss_C|SetCanSwitchTarget false (Variables|Default|GetFortControllerComponent_AIRandomTargeter_KorBoss)))

(fn ConstructionScript
  (|CallParentFunction))
