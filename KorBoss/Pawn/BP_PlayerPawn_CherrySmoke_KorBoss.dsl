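; Server-driven pawn for the Kor boss, handling cloaking, forced leaps, aggro switching and a health-triggered second phase.
; Accumulated damage breaks cloak or, at a team-size-scaled threshold, forces a leap and reload.
; In phase two while threatened, a timer finds nearby grounded players, runs EQS, and spawns henchmen via the Living World manager, hostile to Kor's target.
; Tuning uses hotfixable scalable floats; cloak state replicates through an OnRep that fires dispatchers.
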(event EventBeginPlay
  (bind _korcontroller (Variables|Default|GetKorController))
  (bind _fortathenanpcpatrolling (Variables|/Game/Athena/AI/NPCs/Base/BP_BotController_NPC_Base.BP_BotController_NPC_Base_C|GetFortAthenaNpcPatrolling _korcontroller))
  (bind _returnvalue (Class|FortAthenaNpcPatrollingComponent|GetPatrolPath _fortathenanpcpatrolling :!class /Script/FortniteAI.FortAthenaNpcPatrollingComponent))
  (bind _returnvalue_2 (Class|FortAthenaPatrolPath|GetPatrolPoint _returnvalue 0))
  ;[0.04346 0.08739 0.15 0.5] BeginPlay
  (|CallParentFunction)
  (Networking|SwitchHasAuthority
    (:Authority
      (bind (_gamestate _playlist) (Class|FortAsyncActionCurrentPlaylistReady|PlaylistDataReadyAsync)
        (:OnReady
          (Variables|Default|SetMaxTeamSize (Math|Integer|Clamp(Integer) (Variables|/Script/FortniteGame.FortPlaylist|GetMaxTeamSize _playlist) 1 4))
          (bind _asbp_bot_controller_boss_kor_boss (Utilities|Casting|CastToBP_BotController_Boss_KorBoss (Pawn|GetController)))
          (bind _output_get (Variables|Default|SetKorController _asbp_bot_controller_boss_kor_boss))
          (AlertLevel|BindEventtoOnAlertLevelChangedEvent _output_get (EventDispatchers|CreateEvent :!event Custom|KorAlertLevelChanged))
          (EventDispatchers|BindEventtoOnDied (delegate Custom|Died) :!class /Script/FortniteGame.FortPawn)
          (GameplayCue|AddGameplayCueOnActor(Looping) self (Variables|GameplayCue|GetAura_GCNL) (Ability|GameplayCue|MakeGameplayCueParameters 0.0))
          (Ability|Tasks|WaitGameplayTagAddToActor 0 "(TagName=\"MovementMode.Falling\")"
            (:then
              (Utilities|IsValid _returnvalue_2
                (:Is%20Valid
                  (Variables|Default|SetSpawnLocation (Utilities|Operators|Add (Transformation|GetActorLocation _returnvalue_2) (Math|Vector|MakeVector 0.0 0.0 (Components|Capsule|GetScaledCapsuleHalfHeight (Variables|Character|GetCapsuleComponent))))))
                (:Is%20Not%20Valid
                  (Variables|Default|SetSpawnLocation (Transformation|GetActorLocation))))
              (bind _returnvalue_3 (Actor|GetActorOfClass "/DemonArenaGameplay/KB/KB_PlayerDetectionSphere.KB_PlayerDetectionSphere_C"))
              (Utilities|IsValid _returnvalue_3
                (:Is%20Valid
                  (Collision|BindEventtoOnActorBeginOverlap _returnvalue_3 (EventDispatchers|CreateEvent :!event Custom|PlayerEnteredDetectionSphere)))
                (:Is%20Not%20Valid
                  _))
              (if (ScalableFloat|GetValueasBool (Variables|Hotfix|GetIsKnockbackImmune_Hf))
                (Class|BPFLAICommonFunctions|AddUniqueGameplayTag self "(TagName=\"Gameplay.Status.KnockbackImmunity\")")))
            (:Added
              (if (and (not (Variables|Default|GetDoFarLeap)) (not (Variables|Default|GetDoShortLeap)))
                (Variables|Default|SetForceLeap true)))))))))

(event Custom|Died
  (params (DamagedActor :type Actor%20Object%20Reference) (Damage :type Float%20%28single-precision%29) (InstigatedBy :type Controller%20Object%20Reference) (DamageCauser :type Actor%20Object%20Reference) (HitLocation :type Vector) (FHitComponent :type Primitive%20Component%20Object%20Reference) (BoneName :type Name) (Momentum :type Vector))
  ;[0.04346 0.08739 0.15 0.5] BeginPlay
  (Components|Activation|Deactivate (Variables|Default|GetAthenaSpecialActor)))

(event Custom|ForceKorTeleport
  ;[0.5469 0.1612 0.002055 0.5] Debug
  (Variables|Default|SetForceTeleport true))

(event Custom|ApplyCloakState
  ;[0.3392 0.3262 0.4323 0.5] CloakState
  (if (and (not (GameplayTags|HasMatchingGameplayTag "(TagName=\"Athena.Gameplay.Status.Cloaked\")")) (ScalableFloat|GetValueasBool (Variables|Hotfix|GetCloakStateEnabled_Hf)))
    (bind _returnvalue (CallFunction|ApplyMaterialOverridewithFlags (Utilities|ResolveSoftReference (Variables|Default|GetFadeMaterialInstance)) (Variables|Default|GetMaterialParametersToCopy) 0.0 7))
    (Variables|Default|SetCloakedMaterialAdd _returnvalue)
    (GameplayCue|AddGameplayCueOnActor(Looping) self (Variables|GameplayCue|GetCloak_GCNL) (Ability|GameplayCue|MakeGameplayCueParameters 0.0))
    (Ability|GameplayEffect|AddLooseGameplayTags self (Variables|Default|GetInvisGameplayTags))
    (Variables|Default|SetKorIsCloaked true)))

(event Custom|TargetSwitchNoNewTarget
  (bind _korcontroller (Variables|Default|GetKorController))
  (bind _fortcontrollercomponent_airandomtargeter_korboss (Variables|/KorBoss/Pawn/BP_BotController_Boss_KorBoss.BP_BotController_Boss_KorBoss_C|GetFortControllerComponent_AIRandomTargeter_KorBoss _korcontroller))
  ;[0.1384 0.0889 0.15 0.5] TargetSwitchNoNewTarget
  (Utilities|FlowControl|RetriggerableDelay (ScalableFloat|GetValueatLevel (Variables|Hotfix|GetTargetSwitchNoNewTargetDelay_Hf)))
  (if (Variables|/KorBoss/Pawn/FortControllerComponent_AIRandomTargeter_KorBoss.FortControllerComponent_AIRandomTargeter_KorBoss_C|GetCanSwitchTarget _fortcontrollercomponent_airandomtargeter_korboss)
    (Variables|/KorBoss/Pawn/FortControllerComponent_AIRandomTargeter_KorBoss.FortControllerComponent_AIRandomTargeter_KorBoss_C|SetCanSwitchTarget false _fortcontrollercomponent_airandomtargeter_korboss)))

(event EventOnDied_Event
  (params (DamagedActor :type Actor%20Object%20Reference) (Damage :type Float%20%28single-precision%29) (InstigatedBy :type Controller%20Object%20Reference) (DamageCauser :type Actor%20Object%20Reference) (HitLocation :type Vector :default "0, 0, 0") (FHitComponent :type Primitive%20Component%20Object%20Reference) (BoneName :type Name :default "None") (Momentum :type Vector :default "0, 0, 0"))
  ;[0.5312 0.001776 0 0.5] OnDied
  (|CallParentFunction DamagedActor Damage InstigatedBy DamageCauser HitLocation FHitComponent BoneName Momentum)
  (Utilities|Time|ClearandInvalidateTimerbyHandle (Variables|Default|GetP2HenchmenSpawnTHandle))
  (if (Math|Random|RandomBoolwithWeight (ScalableFloat|GetValueatLevel (Variables|Hotfix|GetEliminatedSpeechOdds_Hf)))
    (Class|BPFLAICommonFunctions|PlaySpeechBubble self "(TagName=\"Athena.AI.SpeechBubbles.CherrySmoke.Kor.Eliminated\")" true true)))

(event Custom|RemoveCloakState
  ;[0.3392 0.3262 0.4323 0.5] CloakState
  (if (GameplayTags|HasMatchingGameplayTag "(TagName=\"Athena.Gameplay.Status.Cloaked\")")
    (GameplayCue|RemoveGameplayCue self (Variables|GameplayCue|GetCloak_GCNL))
    (CallFunction|RemoveMaterialOverride (Variables|Default|GetCloakedMaterialAdd))
    (Ability|GameplayEffect|RemoveLooseGameplayTags self (Variables|Default|GetInvisGameplayTags))
    (Variables|Default|SetKorIsCloaked false)))

(event EventAiTargetChanged
  (params (OldTarget :type Actor%20Object%20Reference) (NewTarget :type Actor%20Object%20Reference))
  ;[0.1268 0.08636 0.15 0.5] AITargetChanged
  (bind _targetactor (Class|BPFLAICommonFunctions|GetTargetActorfromBlackboard self))
  (bind _asfort_player_pawn (Utilities|Casting|CastToFortPlayerPawn _targetactor))
  (Variables|Default|SetKorTargetPlayer _asfort_player_pawn))

(event Custom|KorAlertLevelChanged
  (params (BotController :type Fort%20Athena%20AIBot%20Controller%20Object%20Reference) (OldAlertLevel :type EAlertLevel%20Enum :default "Unaware") (NewAlertLevel :type EAlertLevel%20Enum :default "Unaware"))
  ;[0.02491 0.08854 0.05469 0.5] AlertLevelChanged
  (switch Utilities|FlowControl|Switch|SwitchonEAlertLevel NewAlertLevel
    (:Unaware
      (Variables|Default|SetDoUnawareIdle true))
    (:Suspicious
      _)
    (:LostTarget
      _)
    (:Threatened
      (Variables|Default|SetDoUnawareIdle false)
      (Ability|Tasks|WaitGameplayTagAddToActor 0 "(TagName=\"KorBoss.IsInPhase2\")"
        (:Added
          (Utilities|FlowControl|Delay (ScalableFloat|GetValueatLevel (Variables|Hotfix|GetP2HenchmenStartSpawnDelay_Hf)))
          (CallFunction|Phase2HenchmenSpawn)
          (bind _returnvalue (Utilities|Time|SetTimerbyEvent (EventDispatchers|CreateEvent :!event Custom|Phase2HenchmenSpawn) (ScalableFloat|GetValueatLevel (Variables|Hotfix|GetP2HenchmenSpawnPeriod_Hf)) true true))
          (Variables|Default|SetP2HenchmenSpawnTHandle _returnvalue)))))
  (if (Utilities|Time|IsValidTimerHandle (Variables|Default|GetP2HenchmenSpawnTHandle))
    (Utilities|Time|ClearandInvalidateTimerbyHandle (Variables|Default|GetP2HenchmenSpawnTHandle))))

(event Custom|Phase2HenchmenSpawn
  ;[0.09671 0.15 0.08118 0.5] SpawnHenchman
  (if (and (> (ScalableFloat|GetValueasInteger (Variables|Hotfix|GetP2HenchmenSpawnNmb_Hf)) 0) (Utilities|Enum|Equal(Enum) (Variables|Default|GetAlertLevel) "Threatened"))
    (bind _foundplayers (CallFunction|SearchforNearbyPlayers))
    (bind _output_get (Variables|Default|SetFound%20Players _foundplayers))
    (if (Utilities|Array|IsNotEmpty _output_get)
      (bind _returnvalue (AI|EQS|RunEQSQuery "/KorBoss/EQS/EQS_FindHenchmanSpawnLoc.EQS_FindHenchmanSpawnLoc" (Utilities|Array|RandomArrayItem (Variables|Default|GetFound%20Players)) "AllMatching"))
      (AI|EQS|BindEventtoOnQueryFinished _returnvalue (delegate Custom|P2HenchmanSpawnQueryFinished)))))

(event Custom|P2HenchmanSpawnQueryFinished
  (params (QueryInstance :type EQS%20Query%20Instance%20Object%20Reference) (QueryStatus :type EEnvQueryStatus%20Enum :default "Processing"))
  ;[0.09671 0.15 0.08118 0.5] SpawnHenchman
  (bind (_resultlocations _returnvalue) (AI|EQS|GetQueryResultsasLocations QueryInstance))
  (if _returnvalue
    (for _ (range (ScalableFloat|GetValueasInteger (Variables|Hotfix|GetP2HenchmenSpawnNmb_Hf)))
      (CallFunction|SpawnHenchmanPointProvider (Utilities|Array|Get(acopy) _resultlocations)))))

(event Custom|SpawnHenchmanPointProvider
  (params (Location :type Vector))
  (bind (_returnvalue_roll _returnvalue_pitch _returnvalue_yaw) (Math|Rotator|BreakRotator (Transformation|GetActorRotation)))
  ;[0.09671 0.15 0.08118 0.5] SpawnHenchman
  (if (Utilities|Operators|NotEqual(!=) :A Location)
    (Class|FortAthenaLivingWorldManager|TrySpawnEvent (Class|FortAthenaLivingWorldManager|GetLivingWorldManager) (Utilities|Struct|MakeDataTableRowHandle "/Lager/Lager/EventTables/Lager_EventTable_Boss_CH6S3.Lager_EventTable_Boss_CH6S3" "KBH_Prefab") (Math|Transform|MakeTransform Location (Math|Rotator|MakeRotator 0.0 0.0 _returnvalue_yaw)) (delegate Custom|ActorSpawned))))

(event Custom|ActorSpawned
  (params (SpawnedActors :type Array%20of%20Actor%20Object%20References) (bSuccess :type Boolean))
  (bind _self self)
  ;[0.09671 0.15 0.08118 0.5] SpawnHenchman
  (for _array_element SpawnedActors
    (bind _asbp_player_pawn_non_participant (Utilities|Casting|CastToBP_PlayerPawn_NonParticipant _array_element))
    (Variables|Default|SetInstigator%20Pawn _self _asbp_player_pawn_non_participant)
    (bind _targetactor (Class|BPFLAICommonFunctions|GetTargetActorfromBlackboard _self))
    (Utilities|IsValid _targetactor
      (:Is%20Valid
        (Affiliation|ChangeAttitudeBetweenActors _asbp_player_pawn_non_participant _targetactor "Hostile")
        (bind _asfort_athena_aibot_controller (Utilities|Casting|CastToFortAthenaAIBotController (Fortnite|GetControllerfromActor _asbp_player_pawn_non_participant)))
        (Class|FortAthenaAIBotController|AddTargetInfo _asfort_athena_aibot_controller _targetactor true "Threat_Seeing" 5.0 2e+03)))))

(event EventOnDamageServer
  (params (Damage :type Float%20%28single-precision%29) (DamageTags :type Gameplay%20Tag%20Container%20Structure%20%28by%20ref%29) (Momentum :type Vector :default "0, 0, 0") (HitInfo :type Hit%20Result%20Structure%20%28by%20ref%29) (InstigatedBy :type Controller%20Object%20Reference) (DamageCauser :type Actor%20Object%20Reference) (EffectContext :type Gameplay%20Effect%20Context%20Handle%20Structure))
  ;[0.15 0.03644 0.05953 0.5] OnDamage
  (|CallParentFunction Damage DamageTags Momentum HitInfo InstigatedBy DamageCauser EffectContext)
  (if (!= DamageCauser self)
    (if (and (<= (/ (FortPawn|GetHealth) (FortPawn|GetMaxHealth)) (ScalableFloat|GetValueatLevel (Variables|Hotfix|GetCloakStatePercentAct_Hf))) (not (GameplayTags|HasMatchingGameplayTag "(TagName=\"KorBoss.IsInPhase2\")")))
      (Class|BPFLAICommonFunctions|AddUniqueGameplayTag self "(TagName=\"KorBoss.IsInPhase2\")")
      (if (Math|Random|RandomBoolwithWeight (ScalableFloat|GetValueatLevel (Variables|Hotfix|GetHalfHealthSpeechOdds_Hf)))
        (Variables|Default|SetPlayHalfHealthSpeechBubble true)))
    (if (not (Variables|Default|GetDmgThresholdReached))
      (bind _output_get (Variables|Default|SetDmgReceived (+ (Variables|Default|GetDmgReceived) Damage)))
      (if (>= _output_get (ScalableFloat|GetValueatLevel (Variables|Hotfix|GetLeapDmgThreashold_Hf) (Math|Conversions|ToFloat(Integer) (Variables|Default|GetMaxTeamSize))))
        (Variables|Default|SetDmgThresholdReached true)
        (Variables|Default|SetForceLeap true)
        (Variables|Default|SetDoFarLeap false)
        (Variables|Default|SetDoShortLeap false)
        (Class|BPFLAICommonFunctions|ReloadPawnWeapon self)
        (elif (and (>= _output_get (ScalableFloat|GetValueatLevel (Variables|Hotfix|GetCloakDmgThreashold_Hf))) (GameplayTags|HasMatchingGameplayTag "(TagName=\"Athena.Gameplay.Status.Cloaked\")"))
          (CallFunction|RemoveCloakState))))))

(event Custom|ResetDamageThreshold
  ;[0.15 0.03644 0.05953 0.5] OnDamage
  (Variables|Default|SetDmgReceived 0.0)
  (Variables|Default|SetDmgThresholdReached false))

(event Custom|PlayerEnteredDetectionSphere
  (params (OverlappedActor :type Actor%20Object%20Reference) (OtherActor :type Actor%20Object%20Reference))
  (bind _asfort_player_pawn (Utilities|Casting|CastToFortPlayerPawn OtherActor))
  (if (not (GameplayTags|HasMatchingGameplayTag _asfort_player_pawn "(TagName=\"Pawn.Athena.NonParticipant\")"))
    (Class|FortAthenaAIBotController|AddTargetInfo (Variables|Default|GetKorController) _asfort_player_pawn false "Threat_Seeing" 5.0 1.5e+04)))

(fn OnRep_KorIsCloaked
  (if (Variables|Default|GetKorIsCloaked)
    (Default|CallKorJustCloaked)
    (else
      (Default|CallKorJustUncloaked))))

(fn SearchForNearbyPlayers
  (bind (_returnvalue_x _returnvalue_y _returnvalue_z) (Math|Vector|BreakVector (Transformation|GetActorLocation)))
  (bind _outhits (Collision|MultiSphereTraceForObjects :Start (Transformation|GetActorLocation) :End (+ (Transformation|GetActorLocation) "0.000000,0.000000,1.000000") :Radius (ScalableFloat|GetValueatLevel (Variables|Hotfix|GetP2PlayerSearchRadius_Hf)) :ObjectTypes (Utilities|Array|MakeArray "ObjectTypeQuery3") :bTraceComplex false :DrawDebugType "None" :bIgnoreSelf true :TraceColor "(R=1.000000,G=0.000000,B=0.000000,A=1.000000)" :TraceHitColor "(R=0.000000,G=1.000000,B=0.000000,A=1.000000)" :DrawTime 1e+01))
  (for _array_element _outhits
    (bind _returnvalue (GameplayTags|HasMatchingGameplayTag(Message) (.HitActor (Collision|BreakHitResult _array_element)) "(TagName=\"Pawn.Athena.NonParticipant\")"))
    (bind _returnvalue_2 (GameplayTags|HasAnyMatchingGameplayTags(Message) (.HitActor (Collision|BreakHitResult _array_element)) (GameplayTags|MakeLiteralGameplayTagContainer "(GameplayTags=((TagName=\"MovementMode.Falling\"),(TagName=\"MovementMode.Floating\"),(TagName=\"MovementMode.Flying\"),(TagName=\"MovementMode.Skydiving\"),(TagName=\"MovementMode.Parachuting\"),(TagName=\"GravWell.BassJump.Activated\"),(TagName=\"MistRaven.GrimStrut.Active\"),(TagName=\"MovementMode.IcicleLarva\"),(TagName=\"MovementMode.CubeGliding\"),(TagName=\"MovementMode.CubeSurfing\"),(TagName=\"Ability.BeastGirl.StormClaws.Leap.Active\")))")))
    (bind (_returnvalue_x_2 _returnvalue_y_2 _returnvalue_z_2) (Math|Vector|BreakVector (Transformation|GetActorLocation (.HitActor (Collision|BreakHitResult _array_element)))))
    (if (Math|Boolean|ANDBoolean (not _returnvalue_2) (not _returnvalue) (<= (Math|Float|Absolute(Float) (- _returnvalue_z_2 _returnvalue_z)) (ScalableFloat|GetValueatLevel (Variables|Hotfix|GetP2PlayerSearchMaxZDist_Hf))))
      (Utilities|Array|Add (LocalVariables|GetFoundPlayers) (.HitActor (Collision|BreakHitResult _array_element)))))
  (return (LocalVariables|GetFoundPlayers)))

(fn TrySwitchAggroTarget
  (Variables|/KorBoss/Pawn/FortControllerComponent_AIRandomTargeter_KorBoss.FortControllerComponent_AIRandomTargeter_KorBoss_C|SetCanSwitchTarget true (Variables|/KorBoss/Pawn/BP_BotController_Boss_KorBoss.BP_BotController_Boss_KorBoss_C|GetFortControllerComponent_AIRandomTargeter_KorBoss (Variables|Default|GetKorController)))
  (CallFunction|TargetSwitchNoNewTarget))

(fn ConstructionScript
  (|CallParentFunction))
